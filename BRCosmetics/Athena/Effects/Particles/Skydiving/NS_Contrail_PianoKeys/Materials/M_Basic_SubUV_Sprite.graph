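# Translucent unlit Niagara sprite material in the PianoKeys skydiving contrail folder, drawing a particle-tinted mask that defaults to a music-notes texture.
# Particle RGB passes through MF_FX_EmissiveAdjustments for the glow color.
# For opacity, particle alpha gets a depth fade, optionally blended toward a view-angle fade, then multiplies the mask's red channel.
# The mobile shading path skips both fades and uses raw particle alpha.
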Material /BRCosmetics/Athena/Effects/Particles/Skydiving/NS_Contrail_PianoKeys/Materials/M_Basic_SubUV_Sprite
graph {
 properties {bUsedWithNiagaraSprites=true, blendMode=BLEND_Translucent, pixelDepthOffsetMode=PDOM_Legacy, shadingModel=MSM_Unlit, usedShadingModels=Unlit}
 nodes [
  n1 = ParticleColor()
  MaterialFunctionCall(Emissive=@n1.RGB) {materialFunction=&MF_FX_EmissiveAdjustments}
  n2 = DepthFade(Opacity=@n1.A, FadeDistance=ScalarParameter() {defaultValue=50, group=Opacity, parameterName=DepthFade})
  n3 = Multiply(
   A=@n2
   B=MaterialFunctionCall("FadeSensitivity (optional)"=ScalarParameter() {group=Opacity, parameterName=Angle Fade Bias}) {materialFunction=&FadeBasedOnViewAngle}.Result
  )
  n4 = ScalarParameter() {group=Opacity, parameterName=Use Angle Fade}
  Multiply(
   A=TextureSampleParameter2D() {parameterName=Mask, texture=/Game/Weapons/FORT_Melee/Effects/Spear_BoomBox/T_FX_MusicNotes_01}.R
   B=ShadingPathSwitch(Default=LinearInterpolate(A=@n2, B=@n3, Alpha=@n4), Mobile=@n1.A)
  )
 ]
}
